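# Physics asset for the F_LRG_Mockingbird player character body, covering its dynamic zipper and necklace bones against the upper spine.
# Capsules on spine_04 and spine_05 form the torso volume, joined by an angularly limited constraint.
# Necklace and two zipper master bodies attach to spine_05 with free swing and twist.
# Each zipper capsule hangs off its master with narrow twist and a wide second swing limit.
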PhysicsAsset /BRCosmetics/Characters/Player/Female/Large/Bodies/F_LRG_Mockingbird/Meshes/F_LRG_Mockingbird_Physics
bodies {
 dyn_necklace_MSTR [{shapeType=Box, extentX=1.01, extentY=1.01, extentZ=1.01, shapeName=dyn_necklace_MSTR_box, rotation=<rot 0 -90 0>}]
 dyn_zipper_02 [{shapeType=Capsule, length=3.4984, radius=1.6626, shapeName=dyn_zipper_02_capsule, center=<1.1005 0 0>, rotation=<rot 90 0 0>}]
 dyn_zipper_02_MSTR [{shapeType=Capsule, length=0.505, radius=0.505, shapeName=dyn_zipper_02_MSTR_capsule, rotation=<rot 0 -90 0>}]
 dyn_zipper_03 [{shapeType=Capsule, length=3.4984, radius=1.6599, shapeName=dyn_zipper_03_capsule, center=<1.1042 0 0>, rotation=<rot 90 0 0>}]
 dyn_zipper_03_MSTR [{shapeType=Capsule, length=0.505, radius=0.505, shapeName=dyn_zipper_03_MSTR_capsule, rotation=<rot 0 -90 0>}]
 spine_04 [{shapeType=Capsule, length=2.75, radius=15.0061, center=<-2.4076 1.4712 0>}]
 spine_05 [{shapeType=Capsule, length=15.131, radius=11.6601, center=<-5.0847 0.908 0>}]
}
constraints [
 {bone1Name=spine_05, bone2Name=spine_04, swing1LimitDegrees=45, swing2LimitDegrees=45, twistLimitDegrees=45}
 {
  bone1Name=dyn_necklace_MSTR
  bone2Name=spine_05
  swing1LimitDegrees=45
  swing1Motion=Free
  swing2LimitDegrees=45
  swing2Motion=Free
  twistLimitDegrees=45
  twistMotion=Free
 }
 {
  bone1Name=dyn_zipper_03_MSTR
  bone2Name=spine_05
  swing1LimitDegrees=45
  swing1Motion=Free
  swing2LimitDegrees=45
  swing2Motion=Free
  twistLimitDegrees=45
  twistMotion=Free
 }
 {bone1Name=dyn_zipper_03, bone2Name=dyn_zipper_03_MSTR, swing1LimitDegrees=45, swing2LimitDegrees=100, twistLimitDegrees=15}
 {
  bone1Name=dyn_zipper_02_MSTR
  bone2Name=spine_05
  swing1LimitDegrees=45
  swing1Motion=Free
  swing2LimitDegrees=45
  swing2Motion=Free
  twistLimitDegrees=45
  twistMotion=Free
 }
 {bone1Name=dyn_zipper_02, bone2Name=dyn_zipper_02_MSTR, swing1LimitDegrees=45, swing2LimitDegrees=100, twistLimitDegrees=15}
]
