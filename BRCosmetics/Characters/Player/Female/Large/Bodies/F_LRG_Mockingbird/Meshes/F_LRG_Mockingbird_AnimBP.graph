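# Post-processing AnimBP for the F_LRG_Mockingbird body, adding secondary motion to zippers, necklace, hood and pouch belt.
# It copies the attached parent's pose through the F_MED Mockingbird deform Control Rig, then chains weighted Constraints, Look Ats and spring AnimDynamics.
# EventGraph derives left-thigh rotation and left-strafe alphas that drive additive belt-bone offsets before a Look At chain aims each segment.
# Several nodes, including a RigidBody, sit unconnected.
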AnimBlueprint /BRCosmetics/Characters/Player/Female/Large/Bodies/F_LRG_Mockingbird/Meshes/F_LRG_Mockingbird_AnimBP
target_skeleton=&/F_LRG_Mockingbird_Skeleton
graphs [
 AnimGraph {nodes [
   Anim Dynamics - Bone: dyn_zipper_03(Alpha=0.25, ExternalForce="0,0,0") {
    angularDampingOverride=1
    angularSpringConstant=300
    bAngularSpring=true
    bGravityOverrideInSimSpace=true
    gravityScale=0.45
    simulationSpace=RootRelative
    alphaBoolBlend {blendOption=Linear}
    alphaScaleBias {scale=2}
    boundBone {boneName=dyn_zipper_03}
    componentAppliedLinearAccClamp=<2500 2500 2500>
    componentLinearAccScale=<1 1 1>
    componentLinearVelScale=<1 1 1>
    physicsBodyDefinitions [{
      sphereCollisionRadius=0
      boundBone {boneName=dyn_zipper_03}
      boxExtents=<10 5 5>
      constraintSetup {angularLimitsMax=<15 35 45>, angularLimitsMin=<-15 -120 -45>, angularTarget=<0 0 0>}
      localJointOffset=<2 0 0>
     }]
    relativeSpaceBone {boneName=root}
   }
   Constraint!(
    ComponentPose
    Alpha=1
    ConstraintWeights_0=0.5773
    ConstraintWeights_1:float=0.1946
    ConstraintWeights_2:float=0.1526
    ConstraintWeights_3:float=0.0009
    ConstraintWeights_4:float=0.0643
    ConstraintWeights_5:float=0.0093
    ConstraintWeights_6:float=0.001
    ConstraintWeights_7:float=0
   ) -> (Pose) {
    alphaBoolBlend {blendOption=Linear}
    constraintSetup [
     {targetBone {boneName=pelvis}}
     {targetBone {boneName=spine_01}}
     {targetBone {boneName=spine_02}}
     {targetBone {boneName=spine_03}}
     {targetBone {boneName=thigh_l}}
     {targetBone {boneName=thigh_twist_01_l}}
     {targetBone {boneName=thigh_r}}
     {targetBone {boneName=thigh_twist_01_r}}
    ]
    constraintWeights [1,1,1,1,1,1,1,1]
    boneToModify=~
   }
   Look At(Alpha=1) {alphaBoolBlend {blendOption=Linear}, boneToModify=~, lookAtTarget=~}
   "Transform (Modify) Bone - Bone: dyn_necklace_MSTR"(Alpha=1, Translation="-0.000,0,0.75", Rotation="0,0,-5", Scale="1,1,1") {
    rotationMode=BMM_Additive
    rotationSpace=BCS_BoneSpace
    translationMode=BMM_Additive
    translationSpace=BCS_BoneSpace
    alphaBoolBlend {blendOption=Linear}
    boneToModify {boneName=dyn_necklace_MSTR}
    rotation {roll=-5}
    translation {z=0.75}
   }
   RigidBody!(ComponentPose, Alpha, OverrideWorldGravity:Vector="0,0,-980", ExternalForce) -> (Pose) {
    bClampLinearTranslationLimitToRefPose=true
    bOverrideWorldGravity=true
    overridePhysicsAsset=&/F_LRG_Mockingbird_Physics
    simulationSpace=BaseBoneSpace
    alphaBoolBlend {blendOption=Linear}
    baseBoneRef {boneName=root}
    componentAppliedLinearAccClamp=<3000 3000 3000>
    componentLinearAccScale=<0.3 0.3 0.5>
    componentLinearVelScale=<0.3 0.3 0.3>
   }
   n1 = Control Rig(Source=Copy Pose From Mesh() {bUseAttachedParent=true}) {
    alphaBoolBlend {blendOption=Linear}
    controlRigAssetReference {blueprintRigClass=&/F_MED_Mockingbird_DeformRig_Master_ControlRig.F_MED_Mockingbird_DeformRig_Master_ControlRig_C}
   }
   n2 = "Constraint - dyn_zipper_01_MSTR (PPPPPPPP)"!(
    ComponentPose=Local To Component(LocalPose=@n1)
    Alpha=1
    ConstraintWeights_0=0.2092
    ConstraintWeights_1:float=0.2682
    ConstraintWeights_2:float=0.2672
    ConstraintWeights_3:float=0.1949
    ConstraintWeights_4:float=0.0532
    ConstraintWeights_5:float=0.002
    ConstraintWeights_6:float=0.0004
    ConstraintWeights_7:float=0.0049
   ) -> (Pose) {
    alphaBoolBlend {blendOption=Linear}
    boneToModify {boneName=dyn_zipper_01_MSTR}
    constraintSetup [
     {targetBone {boneName=pelvis}}
     {targetBone {boneName=spine_01}}
     {targetBone {boneName=spine_02}}
     {targetBone {boneName=spine_03}}
     {targetBone {boneName=spine_04}}
     {targetBone {boneName=spine_05}}
     {targetBone {boneName=clavicle_l}}
     {targetBone {boneName=thigh_l}}
    ]
    constraintWeights [1,1,1,1,1,1,1,1]
   }
   n3 = "Constraint - dyn_zipper_02_MSTR (PPPPP)"!(
    ComponentPose=@n2
    Alpha=1
    ConstraintWeights_0=0.0763
    ConstraintWeights_1:float=0.6059
    ConstraintWeights_2:float=0.0517
    ConstraintWeights_3:float=0.2584
    ConstraintWeights_4:float=0.0076
   ) -> (Pose) {
    alphaBoolBlend {blendOption=Linear}
    boneToModify {boneName=dyn_zipper_02_MSTR}
    constraintSetup [
     {targetBone {boneName=upperarm_r}}
     {targetBone {boneName=lowerarm_r}}
     {targetBone {boneName=lowerarm_twist_01_r}}
     {targetBone {boneName=lowerarm_twist_02_r}}
     {targetBone {boneName=upperarm_twist_02_r}}
    ]
    constraintWeights [1,1,1,1,1]
   }
   n4 = "Constraint - dyn_zipper_03_MSTR (PPP)"!(ComponentPose=@n3, Alpha=1, ConstraintWeights_0=0.5342, ConstraintWeights_1:float=0.0887, ConstraintWeights_2:float=0.3771) -> (Pose) {
    alphaBoolBlend {blendOption=Linear}
    boneToModify {boneName=dyn_zipper_03_MSTR}
    constraintSetup [{targetBone {boneName=lowerarm_l}}, {targetBone {boneName=lowerarm_twist_01_l}}, {targetBone {boneName=lowerarm_twist_02_l}}]
    constraintWeights [1,1,1]
   }
   n5 = Anim Dynamics - Bone: dyn_zipper_01(ComponentPose=@n4, Alpha=0.25, ExternalForce="0,0,0") {
    angularDampingOverride=1
    angularSpringConstant=1500
    bAngularSpring=true
    bGravityOverrideInSimSpace=true
    bOverrideAngularDamping=true
    gravityScale=0.45
    simulationSpace=RootRelative
    alphaBoolBlend {blendOption=Linear}
    alphaScaleBias {scale=2}
    boundBone {boneName=dyn_zipper_01}
    componentAppliedLinearAccClamp=<2500 2500 2500>
    componentLinearAccScale=<1 1 1>
    componentLinearVelScale=<1 1 1>
    physicsBodyDefinitions [{
      sphereCollisionRadius=0
      boundBone {boneName=dyn_zipper_01}
      boxExtents=<10 5 5>
      constraintSetup {angularLimitsMax=<15 35 15>, angularLimitsMin=<-15 -35 -15>, angularTarget=<0 0 0>}
      localJointOffset=<2 0 0>
     }]
    relativeSpaceBone {boneName=root}
   }
   n6 = Anim Dynamics - Bone: dyn_zipper_02(ComponentPose=@n5, Alpha=1, ExternalForce="0,0,0") {
    angularDampingOverride=1
    angularSpringConstant=300
    bAngularSpring=true
    bGravityOverrideInSimSpace=true
    bOverrideAngularDamping=true
    gravityScale=0.45
    simulationSpace=RootRelative
    alphaBoolBlend {blendOption=Linear}
    boundBone {boneName=dyn_zipper_02}
    componentAppliedLinearAccClamp=<2500 2500 2500>
    componentLinearAccScale=<1 1 1>
    componentLinearVelScale=<1 1 1>
    physicsBodyDefinitions [{
      sphereCollisionRadius=0
      boundBone {boneName=dyn_zipper_02}
      boxExtents=<10 5 5>
      constraintSetup {angularLimitsMax=<5 25 60>, angularLimitsMin=<-5 -65 -60>, angularTarget=<0 0 0>}
      localJointOffset=<2 0 0>
     }]
    relativeSpaceBone {boneName=root}
   }
   n7 = Anim Dynamics - Bone: dyn_zipper_03(ComponentPose=@n6, Alpha=1, ExternalForce="0,0,0") {
    angularDampingOverride=1
    angularSpringConstant=300
    bAngularSpring=true
    bGravityOverrideInSimSpace=true
    bOverrideAngularDamping=true
    gravityScale=0.45
    simulationSpace=RootRelative
    alphaBoolBlend {blendOption=Linear}
    boundBone {boneName=dyn_zipper_03}
    componentAppliedLinearAccClamp=<2500 2500 2500>
    componentLinearAccScale=<1 1 1>
    componentLinearVelScale=<1 1 1>
    physicsBodyDefinitions [{
      sphereCollisionRadius=0
      boundBone {boneName=dyn_zipper_03}
      boxExtents=<10 5 5>
      constraintSetup {angularLimitsMax=<5 25 60>, angularLimitsMin=<-5 -65 -60>, angularTarget=<0 0 0>}
      localJointOffset=<2 0 0>
     }]
    relativeSpaceBone {boneName=root}
   }
   n8 = "Constraint - dyn_necklace_Lookat_MSTR (PPPPPPP)"!(
    ComponentPose=@n7
    Alpha=1
    ConstraintWeights_0=0.0174
    ConstraintWeights_1:float=0.1612
    ConstraintWeights_2:float=0.6733
    ConstraintWeights_3:float=0.0102
    ConstraintWeights_4:float=0.0097
    ConstraintWeights_5:float=0.0824
    ConstraintWeights_6:float=0.0457
   ) -> (Pose) {
    alphaBoolBlend {blendOption=Linear}
    boneToModify {boneName=dyn_necklace_Lookat_MSTR}
    constraintSetup [
     {targetBone {boneName=spine_03}}
     {targetBone {boneName=spine_04}}
     {targetBone {boneName=spine_05}}
     {targetBone {boneName=clavicle_l}}
     {targetBone {boneName=clavicle_r}}
     {targetBone {boneName=neck_01}}
     {targetBone {boneName=neck_02}}
    ]
    constraintWeights [1,1,1,1,1,1,1]
   }
   n9 = "Constraint - dyn_necklace_TGT1 (PPPPP)"!(
    ComponentPose=@n8
    Alpha=1
    ConstraintWeights_0=0
    ConstraintWeights_1:float=0.0728
    ConstraintWeights_2:float=0.3497
    ConstraintWeights_3:float=0.3712
    ConstraintWeights_4:float=0.2063
   ) -> (Pose) {
    alphaBoolBlend {blendOption=Linear}
    boneToModify {boneName=dyn_necklace_TGT1}
    constraintSetup [
     {targetBone {boneName=spine_02}}
     {targetBone {boneName=spine_03}}
     {targetBone {boneName=spine_04}}
     {targetBone {boneName=spine_05}}
     {targetBone {boneName=clavicle_l}}
    ]
    constraintWeights [1,1,1,1,1]
   }
   n10 = Look At - Bone: dyn_necklace_Lookat_MSTR(ComponentPose=@n9, Alpha=1) {
    alphaBoolBlend {blendOption=Linear}
    boneToModify {boneName=dyn_necklace_Lookat_MSTR}
    lookAtLocation {x=0}
    lookAtTarget {boneReference {boneName=dyn_necklace_TGT1}}
    lookAt_Axis {axis=<1 0>}
   }
   n11 = Anim Dynamics - Bone: dyn_necklace(ComponentPose=@n10, Alpha=0.5, ExternalForce="0,0,0") {
    angularDampingOverride=0.998
    angularSpringConstant=500
    bAngularSpring=true
    bOverrideAngularDamping=true
    bUseGravityOverride=true
    simulationSpace=RootRelative
    alphaBoolBlend {blendOption=Linear}
    boundBone {boneName=dyn_necklace}
    componentLinearAccScale=<0.3 0.3 0.3>
    componentLinearVelScale=<0.3 0.3 0.3>
    gravityOverride {z=-980}
    physicsBodyDefinitions [{
      boundBone {boneName=dyn_necklace}
      boxExtents=<7 12 12>
      constraintSetup {angularLimitsMax=<1 15 0>, angularLimitsMin=<-1 -15 -35>, angularTarget=<0 0 0>}
      localJointOffset=<3 0 0>
     }]
    relativeSpaceBone {boneName=pelvis}
   }
   n12 = Anim Dynamics - Bone: dyn_necklace_01(ComponentPose=@n11, Alpha=0.5, ExternalForce="0,0,0") {
    angularDampingOverride=0.998
    angularSpringConstant=300
    bAngularSpring=true
    bOverrideAngularDamping=true
    bUseGravityOverride=true
    simulationSpace=RootRelative
    alphaBoolBlend {blendOption=Linear}
    boundBone {boneName=dyn_necklace_01}
    componentLinearAccScale=<0.3 0.3 0.3>
    componentLinearVelScale=<0.3 0.3 0.3>
    gravityOverride {z=-980}
    physicsBodyDefinitions [{
      boundBone {boneName=dyn_necklace_01}
      boxExtents=<15 15 15>
      constraintSetup {angularLimitsMax=<1 15 0>, angularLimitsMin=<-1 -15 -35>, angularTarget=<0 0 0>}
      localJointOffset=<5 0 0>
     }]
    relativeSpaceBone {boneName=pelvis}
   }
   n13 = Anim Dynamics - Bone: dyn_hood(ComponentPose=@n12, Alpha=0.75, ExternalForce="0,0,0") {
    angularDampingOverride=0.998
    angularSpringConstant=75
    bAngularSpring=true
    bGravityOverrideInSimSpace=true
    bOverrideAngularDamping=true
    bUseGravityOverride=true
    bUsePlanarLimit=false
    simulationSpace=RootRelative
    alphaBoolBlend {blendOption=Linear}
    boundBone {boneName=dyn_hood}
    componentAppliedLinearAccClamp=<3000 3000 3000>
    componentLinearAccScale=<0.3 0.3 0.3>
    componentLinearVelScale=<0.3 0.3 0.3>
    gravityOverride {z=-980}
    physicsBodyDefinitions [{
      collisionType=InnerSphere
      sphereCollisionRadius=1
      boundBone {boneName=dyn_hood}
      boxExtents=<7 12 12>
      constraintSetup {twistAxis=AxisZ, angularLimitsMax=<0 15 5>, angularLimitsMin=<0 -15 -35>, angularTarget=<0 0 0>}
      localJointOffset=<-1 0 0>
     }]
    relativeSpaceBone {boneName=root}
   }
   n14 = "Constraint - dyn_pouch_belt_01_MSTR (PPPPPPPP)"!(
    ComponentPose=@n13
    Alpha=1
    ConstraintWeights_0=0.2762
    ConstraintWeights_1:float=0.2856
    ConstraintWeights_2:float=0.2586
    ConstraintWeights_3:float=0.1726
    ConstraintWeights_4:float=0.0046
    ConstraintWeights_5:float=0.0014
    ConstraintWeights_6:float=0.0002
    ConstraintWeights_7:float=0.0007
   ) -> (Pose) {
    alphaBoolBlend {blendOption=Linear}
    boneToModify {boneName=dyn_pouch_belt_01_MSTR}
    constraintSetup [
     {targetBone {boneName=pelvis}}
     {targetBone {boneName=spine_01}}
     {targetBone {boneName=spine_02}}
     {targetBone {boneName=spine_03}}
     {targetBone {boneName=spine_04}}
     {targetBone {boneName=thigh_l}}
     {targetBone {boneName=thigh_twist_01_l}}
     {targetBone {boneName=thigh_r}}
    ]
    constraintWeights [1,1,1,1,1,1,1,1]
   }
   n15 = "Constraint - dyn_pouch_belt_02_MSTR (PPPPPPPPP)"!(
    ComponentPose=@n14
    Alpha=1
    ConstraintWeights_0=0.525
    ConstraintWeights_1:float=0.2295
    ConstraintWeights_2:float=0.1804
    ConstraintWeights_3:float=0.0436
    ConstraintWeights_4:float=0.0012
    ConstraintWeights_5:float=0.0165
    ConstraintWeights_6:float=0.0027
    ConstraintWeights_7:float=0.001
    ConstraintWeights_8:float=0
   ) -> (Pose) {
    alphaBoolBlend {blendOption=Linear}
    boneToModify {boneName=dyn_pouch_belt_02_MSTR}
    constraintSetup [
     {targetBone {boneName=pelvis}}
     {targetBone {boneName=spine_01}}
     {targetBone {boneName=spine_02}}
     {targetBone {boneName=spine_03}}
     {targetBone {boneName=spine_04}}
     {targetBone {boneName=thigh_l}}
     {targetBone {boneName=thigh_twist_01_l}}
     {targetBone {boneName=thigh_r}}
     {targetBone {boneName=thigh_twist_01_r}}
    ]
    constraintWeights [1,1,1,1,1,1,1,1,1]
   }
   n16 = "Constraint - dyn_pouch_belt_03_MSTR (PPPPPPPP)"!(
    ComponentPose=@n15
    Alpha=1
    ConstraintWeights_0=0.5771
    ConstraintWeights_1:float=0.1819
    ConstraintWeights_2:float=0.1423
    ConstraintWeights_3:float=0
    ConstraintWeights_4:float=0.0839
    ConstraintWeights_5:float=0.0148
    ConstraintWeights_6:float=0
    ConstraintWeights_7:float=0
   ) -> (Pose) {
    alphaBoolBlend {blendOption=Linear}
    boneToModify {boneName=dyn_pouch_belt_03_MSTR}
    constraintSetup [
     {targetBone {boneName=pelvis}}
     {targetBone {boneName=spine_01}}
     {targetBone {boneName=spine_02}}
     {targetBone {boneName=spine_03}}
     {targetBone {boneName=thigh_l}}
     {targetBone {boneName=thigh_twist_01_l}}
     {targetBone {boneName=thigh_r}}
     {targetBone {boneName=thigh_twist_01_r}}
    ]
    constraintWeights [1,1,1,1,1,1,1,1]
   }
   n17 = "Constraint - dyn_pouch_belt_04_MSTR (PPPPPPPP)"!(
    ComponentPose=@n16
    Alpha=1
    ConstraintWeights_0=0.5542
    ConstraintWeights_1:float=0.0494
    ConstraintWeights_2:float=0.016
    ConstraintWeights_3:float=0
    ConstraintWeights_4:float=0.253
    ConstraintWeights_5:float=0.1261
    ConstraintWeights_6:float=0.0012
    ConstraintWeights_7:float=0
   ) -> (Pose) {
    alphaBoolBlend {blendOption=Linear}
    boneToModify {boneName=dyn_pouch_belt_04_MSTR}
    constraintSetup [
     {targetBone {boneName=pelvis}}
     {targetBone {boneName=spine_01}}
     {targetBone {boneName=spine_02}}
     {targetBone {boneName=spine_03}}
     {targetBone {boneName=thigh_l}}
     {targetBone {boneName=thigh_twist_01_l}}
     {targetBone {boneName=thigh_r}}
     {targetBone {boneName=thigh_twist_01_r}}
    ]
    constraintWeights [1,1,1,1,1,1,1,1]
   }
   n18 = "Constraint - dyn_pouch_tgt (PPPPPPPP)"!(
    ComponentPose=@n17
    Alpha=1
    ConstraintWeights_0=0.31
    ConstraintWeights_1:float=0.0363
    ConstraintWeights_2:float=0.0184
    ConstraintWeights_3:float=0.0001
    ConstraintWeights_4:float=0.5216
    ConstraintWeights_5:float=0.1131
    ConstraintWeights_6:float=0.0005
    ConstraintWeights_7:float=0
   ) -> (Pose) {
    alphaBoolBlend {blendOption=Linear}
    boneToModify {boneName=dyn_pouch_tgt}
    constraintSetup [
     {targetBone {boneName=pelvis}}
     {targetBone {boneName=spine_01}}
     {targetBone {boneName=spine_02}}
     {targetBone {boneName=spine_03}}
     {targetBone {boneName=thigh_l}}
     {targetBone {boneName=thigh_twist_01_l}}
     {targetBone {boneName=thigh_r}}
     {targetBone {boneName=thigh_twist_01_r}}
    ]
    constraintWeights [1,1,1,1,1,1,1,1]
   }
   n19 = Get ThighRot l() -> (ThighRot l:double)
   n20 = "Transform (Modify) Bone - Bone: dyn_pouch_belt_02"(ComponentPose=@n18, Alpha=@n19, Translation="0.341,-0.908,1.578", Rotation="0,0,0", Scale="1,1,1") {
    rotationMode=BMM_Additive
    rotationSpace=BCS_BoneSpace
    translationMode=BMM_Additive
    translationSpace=BCS_BoneSpace
    alphaBoolBlend {blendOption=Linear}
    boneToModify {boneName=dyn_pouch_belt_02}
    translation=<0.341 -0.908 1.578>
   }
   n21 = "Transform (Modify) Bone - Bone: dyn_pouch_belt_03"(ComponentPose=@n20, Alpha=@n19, Translation="-0.032,-3.784,0.270", Rotation="0,0,0", Scale="1,1,1") {
    rotationMode=BMM_Additive
    rotationSpace=BCS_BoneSpace
    translationMode=BMM_Additive
    translationSpace=BCS_BoneSpace
    alphaBoolBlend {blendOption=Linear}
    boneToModify {boneName=dyn_pouch_belt_03}
    translation=<-0.032 -3.784 0.27>
   }
   n22 = "Transform (Modify) Bone - Bone: dyn_pouch_belt_04"(ComponentPose=@n21, Alpha=@n19, Translation="0,-3.187,1.962", Rotation="0,0,0", Scale="1,1,1") {
    rotationMode=BMM_Additive
    rotationSpace=BCS_BoneSpace
    translationMode=BMM_Additive
    translationSpace=BCS_BoneSpace
    alphaBoolBlend {blendOption=Linear}
    boneToModify {boneName=dyn_pouch_belt_04}
    translation {y=-3.187, z=1.962}
   }
   n23 = "Transform (Modify) Bone - Bone: dyn_pouch_belt_04"(ComponentPose=@n22, Alpha=Get Strafe left() -> (Strafe left:double), Translation="3,0,5", Rotation="0,0,0", Scale="1,1,1") {
    rotationMode=BMM_Additive
    rotationSpace=BCS_BoneSpace
    translationMode=BMM_Additive
    translationSpace=BCS_BoneSpace
    alphaBoolBlend {blendOption=Linear}
    alphaScaleBiasClamp {bInterpResult=true}
    boneToModify {boneName=dyn_pouch_belt_04}
    translation {x=3, z=5}
   }
   n24 = Look At - Bone: dyn_pouch_belt_01(ComponentPose=@n23, Alpha=1) {
    alphaBoolBlend {blendOption=Linear}
    boneToModify {boneName=dyn_pouch_belt_01}
    lookAtLocation {x=0}
    lookAtTarget {boneReference {boneName=dyn_pouch_belt_02}}
    lookAt_Axis {axis=<1 0>}
   }
   n25 = Look At - Bone: dyn_pouch_belt_02(ComponentPose=@n24, Alpha=1) {
    alphaBoolBlend {blendOption=Linear}
    boneToModify {boneName=dyn_pouch_belt_02}
    lookAtLocation {x=0}
    lookAtTarget {boneReference {boneName=dyn_pouch_belt_03}}
    lookAt_Axis {axis=<1 0>}
   }
   n26 = Look At - Bone: dyn_pouch_belt_03(ComponentPose=@n25, Alpha=1) {
    alphaBoolBlend {blendOption=Linear}
    boneToModify {boneName=dyn_pouch_belt_03}
    lookAtLocation {x=0}
    lookAtTarget {boneReference {boneName=dyn_pouch_belt_04}}
    lookAt_Axis {axis=<1 0>}
   }
   n27 = Look At - Bone: dyn_pouch_belt_04(ComponentPose=@n26, Alpha=1) {
    alphaBoolBlend {blendOption=Linear}
    boneToModify {boneName=dyn_pouch_belt_04}
    lookAtLocation {x=0}
    lookAtTarget {boneReference {boneName=dyn_pouch_tgt}}
    lookAt_Axis {axis=<1 0>}
   }
   Output Pose(Result=Component To Local(ComponentPose=@n27)) {layerGroup=None}
  ]}
 EventGraph {nodes [
   Event BlueprintUpdateAnimation()
   > n1 = Sequence()
   > Set ThighRot l(ThighRot l:double=GetAlphaFromReferenceTransformDelta(SkeletalMeshComponent=Get Skel_Mesh(), BoneName=thigh_l, BaseName=pelvis, InputRange=(X=-20,Y=-100), TwistAxis="1,0,0")) -> (Output_Get:double)
   n2 = PlayerMovement(execute=@n1.then_1, AnimInstance=Self-Reference(), XDotProductCoefficient=1, YDotProductCoefficient=1, ZDotProductCoefficient=1)
   > Set Strafe left(Strafe left:double=MapRangeClamped(Value=@n2, InRangeA=0, InRangeB=-1)) -> (Output_Get:double)
   Event BlueprintInitializeAnimation()
   > Set Skel_Mesh(Skel_Mesh=GetOwnerSkeletalMeshComponent())
  ]}
]
variables [
 Character Part Anim Instance {type=CustomCharacterPartAnimInstance, default=null}
 Skel_Mesh {type=SkeletalMeshComponent}
 Strafe left {type=number}
 ThighRot l {type=number}
]
